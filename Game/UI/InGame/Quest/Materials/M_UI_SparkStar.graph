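# M_UI_SparkStar is an additive-blended UI-domain material for a star spark, filed under the in-game quest UI.
# Its graph holds a single texture sample of T_Spark_Star, a texture borrowed from the LaunchTheBalloon primary mission folder.
# No parameters, animation or output connections appear in the graph.
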Material /Game/UI/InGame/Quest/Materials/M_UI_SparkStar
graph {
 properties {bUsedWithStaticMesh=false, blendMode=BLEND_Additive, materialDomain=MD_UI, pixelDepthOffsetMode=PDOM_Legacy, usedShadingModels=Default Lit}
 nodes [TextureSample() {texture=/Game/Missions/Primary/LaunchTheBalloon/Textures/T_Spark_Star}]
}
